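# Additive unlit Niagara material for mesh and sprite particles that draws a camera lens effect without depth testing.
# Particle color feeds emissive adjustments, while particle alpha scales an opacity mask.
# The mask multiplies a sharpened, inverted screen-space radial gradient, favoring screen edges, by a UV-V wipe; dynamic parameters drive radius and wipe.
# A camera-minus-actor position offset, apparently for world position offset, shifts the mesh onto the camera.
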Material /Game/Effects/Niagara/MasterMaterials/General/M_Camera_Lens_Mesh
graph {
 properties {
  bDisableDepthTest=true
  bUsedWithNiagaraMeshParticles=true
  bUsedWithNiagaraSprites=true
  blendMode=BLEND_Additive
  pixelDepthOffsetMode=PDOM_Legacy
  shadingModel=MSM_Unlit
  usedShadingModels=Unlit
 }
 nodes [
  n1 = ParticleColor()
  MaterialFunctionCall(Emissive=@n1.RGB) {materialFunction=&MF_FX_EmissiveAdjustments}
  n2 = DynamicParameter() {defaultValue=<rgba 0.25 1 0.5 1>, paramNames [Wipe Mask, PatternTextureMult, Radius, Param4]}
  n3 = Multiply(A=@n2.Wipe Mask) {constB=1.5}
  OneMinus(None=@n3)
  n4 = OneMinus(None=MaterialFunctionCall(UVs=ScreenPosition().ViewportUV, Radius=@n2.Radius, Density=ScalarParameter() {defaultValue=2.33, parameterName=density}) {materialFunction=&RadialGradientExponential}.RadialGradientExponential)
  n5 = Power(Base=@n4) {constExponent=12}
  n6 = Add(A=Saturate(None=Multiply(A=@n5) {constB=50}), B=Multiply(A=Power(Base=@n4) {constExponent=5}) {constB=50})
  n7 = OneMinus(None=ComponentMask(None=TextureCoordinate()) {g=true})
  n8 = Saturate(None=Divide(A=Subtract(A=@n7, B=@n3), B=Constant() {desc=0.5, r=0.25}) {constB=0.5})
  Saturate(None=Multiply(A=@n1.A, B=Multiply(A=@n6, B=@n8)))
  n9 = CameraPositionWS()
  n10 = WorldPosition()
  Subtract(A=Subtract(A=Add(A=@n9, B=@n10.XYZ), B=@n10.XYZ), B=ActorPositionWS())
 ]
}
