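# Material function that adds panning distortion-map offsets to FX texture UVs carried in material attributes.
# MF_NFX_Bridge supplies pan time and strength from Niagara; only red and U strength drive both offset axes.
# Per-channel influences scale additions to CustomizedUV0/BentNormal and CustomizedUV1/SubsurfaceColor, the latter of each pair also receiving blue.
# A static switch bypasses the attribute edits; quality and shading-path switches feed no output, despite the description.
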MaterialFunction /Game/Effects/Niagara/MasterMaterials/Functions/MF_FX_Distortion
graph {
 properties {
  bExposeToLibrary=true
  description=Distortion added to FX Textures. Includes embeded quality and shading switches.
  libraryCategoriesText [FX Distortion]
 }
 nodes [
  n1 = GetMaterialAttributes(None=FunctionInput() {inputType=FunctionInput_MaterialAttributes}) {attributeGetTypes [#1]}
  n2 = GetMaterialAttributes(None=@n1.MaterialAttributes) {attributeGetTypes [#1, #2, #3]}
  n3 = ScalarParameter() {
   defaultValue=1
   desc=How much the distortion should add to Channel UV0.
   group="1.2) Distortion UV Influence"
   parameterName=UV0 Distortion Influence
   sliderMax=1
  }
  n4 = LinearInterpolate(
   A=TextureCoordinate()
   B=TextureCoordinate() {coordinateIndex=1}
   Alpha=ScalarParameter() {group="0.1) Distortion UVs", parameterName="UV Channel (Distortion)", sliderMax=1}
  )
  n5 = AppendVector(
   A=ScalarParameter() {defaultValue=1, group="0.1) Distortion UVs", parameterName="U Freq Mult (Distort)", sortPriority=1}
   B=ScalarParameter() {defaultValue=1, group="0.1) Distortion UVs", parameterName="V Freq Mult (Distort)", sortPriority=2}
  )
  n6 = ScalarParameter() {defaultValue=1, group="0.1) Distortion UVs", parameterName="Pan U Speed (Distort)", sortPriority=3}
  n7 = ScalarParameter() {group="0.1) Distortion UVs", parameterName="Pan V Speed (Distort)", sortPriority=4}
  n8 = TextureSampleParameter2D(UVs=Panner(
    Coordinate=Multiply(A=@n4, B=@n5) {desc=This might need a frac option for getting textures to tile. This scales the texture.}
    Time=MaterialFunctionCall() {desc=This is to communicate with Niagara., materialFunction=&/MF_NFX_Bridge}.Distortion UV Time
    Speed=AppendVector(A=@n6, B=@n7)
   ) {desc=This offsets the texture.}) {
   group="0.1) Distortion"
   parameterName=DistortionMap
   samplerType=SAMPLERTYPE_Normal
   sortPriority=1
   texture=&T_Weap_Distort_N_02
  }
  n9 = MaterialFunctionCall() {materialFunction=&MF_NFX_Bridge}
  n10 = Multiply(A=Subtract(A=AppendVector(A=@n8.R, B=@n8.R)) {constB=0.5}, B=AppendVector(A=@n9.Distortion U Strength, B=@n9.Distortion U Strength))
  n11 = Multiply(A=@n3, B=@n10)
  n12 = AppendVector(A=@n10, B=@n8.B)
  n13 = Multiply(A=@n3, B=@n12)
  n14 = GetMaterialAttributes(None=SetMaterialAttributes!(MaterialAttributes=@n2.MaterialAttributes, CustomizedUV0=Add(A=@n2.CustomizedUV0, B=@n11), BentNormal=Add(A=@n2.BentNormal, B=@n13)) {attributeSetTypes [#1, #3]}) {attributeGetTypes [#4, #5]}
  n15 = ScalarParameter() {
   defaultValue=1
   desc=How much the distortion should add to Channel UV1.
   group="1.2) Distortion UV Influence"
   parameterName=UV1 Distortion Influence
   sliderMax=1
   sortPriority=1
  }
  n16 = SetMaterialAttributes!(
   MaterialAttributes=@n14.MaterialAttributes
   CustomizedUV1=Add(A=@n14.CustomizedUV1, B=Multiply(A=@n10, B=@n15))
   SubsurfaceColor=Add(A=@n14.SubsurfaceColor, B=Multiply(A=@n12, B=@n15))
  ) {attributeSetTypes [#4, #5]}
  n17 = StaticSwitchParameter(True=@n16, False=@n1.MaterialAttributes) {group="1) Distortion", parameterName=Distortion}
  FunctionOutput(None=@n17) {outputName="UVs (MA)"}
  FunctionOutput(None=@n10) {outputName="Distortion Value (Vector2)"}
  QualitySwitch()
  ShadingPathSwitch(Default=@n17, Deferred=@n16, Forward=@n16, Mobile=@n1.MaterialAttributes)
 ]
}
