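# Material function for Niagara FX materials that outputs a two-component distortion offset for perturbing texture UVs.
# A parameter blends UV channels 0 and 1, which are frequency-scaled and panned along U by the Niagara bridge's Distortion UV Time.
# The normal-sampled distortion map's red channel drives both components, offset by 0.5 and scaled by the bridge's U strength on both axes.
# The description mentions quality and shading switches, but the graph contains none.
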MaterialFunction /Game/Effects/Niagara/MasterMaterials/Functions/MF_FX_Distortion_v2
graph {
 properties {
  bExposeToLibrary=true
  description=Distortion added to FX Textures. Includes embeded quality and shading switches.
  libraryCategoriesText [FX Distortion]
 }
 nodes [
  n1 = LinearInterpolate(
   A=TextureCoordinate()
   B=TextureCoordinate() {coordinateIndex=1}
   Alpha=ScalarParameter() {group="0.1) Distortion UVs", parameterName="UV Channel (Distortion)", sliderMax=1}
  )
  n2 = AppendVector(
   A=ScalarParameter() {defaultValue=1, group="0.1) Distortion UVs", parameterName="U Freq Mult (Distort)", sortPriority=1}
   B=ScalarParameter() {defaultValue=1, group="0.1) Distortion UVs", parameterName="V Freq Mult (Distort)", sortPriority=2}
  )
  n3 = MaterialFunctionCall() {
   desc="This is to communicate with Niagara. \r\nUse a Niagara expression to offset and control rate."
   materialFunction=&/MF_NFX_Bridge
  }
  n4 = TextureSampleParameter2D(UVs=Panner(
    Coordinate=Multiply(A=@n1, B=@n2) {desc=This might need a frac option for getting textures to tile. This scales the texture.}
    Time=@n3.Distortion UV Time
   ) {desc=This offsets the texture., speedX=1}) {
   group="0.1) Distortion"
   parameterName=DistortionMap
   samplerType=SAMPLERTYPE_Normal
   sortPriority=1
   texture=&T_Weap_Distort_N_02
  }
  n5 = AppendVector(A=@n4.R, B=@n4.R)
  n6 = MaterialFunctionCall() {materialFunction=&MF_NFX_Bridge}
  FunctionOutput(None=Multiply(A=Subtract(A=@n5) {constB=0.5}, B=AppendVector(A=@n6.Distortion U Strength, B=@n6.Distortion U Strength))) {outputName="Distortion Value (Vector2)"}
 ]
}
