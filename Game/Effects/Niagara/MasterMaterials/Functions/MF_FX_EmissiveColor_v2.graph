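# Niagara FX material function that tints BaseColor by Particle Color and derives an HDR emissive color from it.
# The tinted BaseColor is written back and also saturated as the LDR input to DeriveHDRfromLDR.
# The HDR tint blends from the HDR_Tint input toward Particle Color via the ParticleColorHDRTint parameter.
# Intensity and desaturation inputs shape the remap, whose RemappedRange output becomes EmissiveColor in the passed-through material attributes.
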MaterialFunction /Game/Effects/Niagara/MasterMaterials/Functions/MF_FX_EmissiveColor_v2
graph {
 properties {bExposeToLibrary=true, description="LDR to HDR, HDR tint is from Particle Color", libraryCategoriesText ["5) FX Emissive Color", ""]}
 nodes [
  n1 = GetMaterialAttributes(None=FunctionInput() {inputType=FunctionInput_MaterialAttributes}) {attributeGetTypes [#1]}
  n2 = ParticleColor()
  n3 = Multiply(A=@n1.BaseColor, B=@n2.RGB)
  n4 = Saturate(None=@n3)
  n5 = LinearInterpolate(
   A=FunctionInput() {bUsePreviewValueAsDefault=true, inputName=HDR_Tint, sortPriority=1, previewValue=<1 1 1 1>}
   B=@n2.RGB
   Alpha=ScalarParameter() {group="1.2) Emissive", parameterName=ParticleColorHDRTint}
  )
  n6 = FunctionInput() {bUsePreviewValueAsDefault=true, inputName=HDR Intensity Multiplier, inputType=FunctionInput_Scalar, sortPriority=2, previewValue=<1 0 0 0>}
  n7 = FunctionInput() {bUsePreviewValueAsDefault=true, inputName=Desaturate Luminance, inputType=FunctionInput_Scalar, sortPriority=3, previewValue=<1 0 0 0>}
  FunctionOutput(None=SetMaterialAttributes!(
    MaterialAttributes=@n1.MaterialAttributes
    EmissiveColor=MaterialFunctionCall(LDR Input=@n4, HDR Tint=@n5, HDR Intensity Multiplier=@n6, Desaturate Luminance=@n7) {desc=remap colors for emissive, materialFunction=&DeriveHDRfromLDR}.RemappedRange
    BaseColor=@n3
   ) {desc=write emissive color, attributeSetTypes [#2, #1]}) {desc=output material attributes}
 ]
}
