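# Material function for Niagara effect master materials that scales and pans texture UVs.
# Input coordinates are either multiplied by separate U and V frequencies or scaled from center using only the U value, blended by the "Scale UVs From Center" parameter.
# A Panner then offsets the result using the pan speed inputs, timed by the UV Time output of MF_NFX_Bridge, which communicates with Niagara.
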MaterialFunction /Game/Effects/Niagara/MasterMaterials/Functions/V2/UVs/MF_FX_NiagaraUVs
graph {nodes [
  n1 = FunctionInput(Preview=TextureCoordinate()) {bUsePreviewValueAsDefault=true, inputName=TextureCoordinates, inputType=FunctionInput_Vector2}
  n2 = FunctionInput() {
   bUsePreviewValueAsDefault=true
   description=This is an independent multiplier on the incoming frequency
   inputName=U Freq Multiplier
   inputType=FunctionInput_Scalar
   sortPriority=2
   previewValue=<1 0 0 1>
  }
  n3 = FunctionInput() {
   bUsePreviewValueAsDefault=true
   description=This is an independent multiplier on the frequency
   inputName=V Freq Multiplier
   inputType=FunctionInput_Scalar
   sortPriority=3
   previewValue=<1 0 0 1>
  }
  n4 = Multiply(A=@n1, B=AppendVector(A=@n2, B=@n3)) {desc=This might need a frac option for getting textures to tile. This scales the texture.}
  n5 = MaterialFunctionCall(UVs=@n1, Texture Scale=@n2) {materialFunction=&ScaleUVsByCenter}
  n6 = ScalarParameter() {group="1.1) Base Color Texture UVs", parameterName=Scale UVs From Center}
  n7 = FunctionInput() {
   bUsePreviewValueAsDefault=true
   description=A multiplier on the speed of the pan.
   inputName=Pan Speed U
   inputType=FunctionInput_Scalar
   sortPriority=4
  }
  n8 = FunctionInput() {bUsePreviewValueAsDefault=true, description=A multiplier on the speed of the pan., inputName=Pan Speed V, sortPriority=5}
  FunctionOutput(None=Panner(
    Coordinate=LinearInterpolate(A=@n4, B=@n5.UVs, Alpha=@n6)
    Time=MaterialFunctionCall() {desc=This is to communicate with Niagara., materialFunction=/Game/Effects/Niagara/MasterMaterials/Functions/MF_NFX_Bridge}.UV Time
    Speed=AppendVector(A=@n7, B=@n8)
   ) {desc=This offsets the texture.})
 ]}
