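# Master translucent unlit material for Niagara ribbon effects, built as a chain of shared material functions passing Material Attributes.
# Panning UVs plus a distortion offset sample the base color texture, which is then component-masked, color-corrected and given HDR emissive tint.
# Opacity, depth fade, parametric bevel and skydive camera fade functions follow.
# Four dynamic parameter sets expose UV, distortion, shaping and channel-mask controls to Niagara.
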Material /Game/Effects/Niagara/MasterMaterials/General/M_Ribbon_Master_v2
graph {
 properties {
  bUseMaterialAttributes=true
  bUsedWithNiagaraRibbons=true
  blendMode=BLEND_Translucent
  pixelDepthOffsetMode=PDOM_Legacy
  shadingModel=MSM_Unlit
  usedShadingModels=Unlit
 }
 nodes [
  DynamicParameter() {paramNames [Base Texture UV Time, Distortion UV Time, Distortion U Strength, Distortion V Strength]}
  DynamicParameter() {parameterIndex=1, defaultValue=<rgba 0 0 1 1>, paramNames [UV Time, Add, PostMult, Pow]}
  DynamicParameter() {
   desc=This talks to the niagara module Material ComponentMask
   parameterIndex=2
   paramNames [R Component Mask, G Component Mask, B Component Mask, A Component Mask]
  }
  DynamicParameter() {parameterIndex=3, paramNames [UV0 U Freq, UV0 V Freq, UV0 U Freq, UV1 V Freq]}
  n1 = MaterialFunctionCall!(
   TextureCoordinates
   U Freq Multiplier=ScalarParameter() {defaultValue=1, group="1.1) Base Color Texture Uvs", parameterName="U Freq (BaseColor)", sortPriority=2}
   V Freq Multiplier=ScalarParameter() {defaultValue=1, group="1.1) Base Color Texture Uvs", parameterName="V Freq (BaseColor)", sortPriority=3}
   Pan Speed U=ScalarParameter() {group="1.1) Base Color Texture Uvs", parameterName="Pan U Speed (BaseColor)", sortPriority=4}
   Pan Speed V=ScalarParameter() {group="1.1) Base Color Texture Uvs", parameterName="Pan V Speed (BaseColor)", sortPriority=5}
  ) {materialFunction=/Game/Effects/Niagara/MasterMaterials/Functions/V2/UVs/MF_FX_NiagaraUVs}
  n2 = MaterialFunctionCall() {materialFunction=/Game/Effects/Niagara/MasterMaterials/Functions/MF_FX_Distortion_v2}
  n3 = MaterialFunctionCall!(In=MaterialFunctionCall!(In, Uvs=Add(A=@n1.Result, B=@n2."Distortion Value (Vector2)"), Base Color Value) {desc=Read texture, materialFunction=/Game/Effects/Niagara/MasterMaterials/Functions/V2/BaseColor/MF_FX_BaseColorTexture_v02}.MA BaseColor) {materialFunction=/Game/Effects/Niagara/MasterMaterials/Functions/MF_NFX_ComponentMask}
  n4 = MaterialFunctionCall!(
   Distortion
   In=MaterialFunctionCall!(
    In=MaterialFunctionCall!(In=@n3."BaseColor (MA) ") {materialFunction=/Game/Effects/Niagara/MasterMaterials/Functions/MF_Niagara_ColorCorrect}."BaseColor (MA)"
    HDR_Tint
    HDR Intensity Multiplier=ScalarParameter() {defaultValue=1, group="1.2) Emissive", parameterName=Emissive HDR Tint Multiplier}
    Desaturate Luminance=ScalarParameter() {defaultValue=1, group="1.2) Emissive", parameterName=HDR Luminance Desaturate}
   ) {materialFunction=/Game/Effects/Niagara/MasterMaterials/Functions/MF_FX_EmissiveColor_v2}.Result
  ) {materialFunction=/Game/Effects/Niagara/MasterMaterials/Functions/MF_FX_Opacity}
  MaterialFunctionCall!(Material Attributes=MaterialFunctionCall!(In=MaterialFunctionCall!(In=@n4.Opacity) {materialFunction=/Game/Effects/Niagara/MasterMaterials/Functions/Opacity/FX_DepthFade}."Opacity (MA)") {materialFunction=/Game/Effects/Niagara/MasterMaterials/Functions/Opacity/MF_Parametric_OpacityBevel}."Opacitys and BaseColor (MA)") {materialFunction=/Game/Effects/Niagara/MasterMaterials/Functions/V2/MF_Skydive_CameraFade_V2}
  MaterialFunctionCall!(In) {materialFunction=/Game/Effects/Niagara/MasterMaterials/Functions/MF_FX_Distortion}
 ]
}
