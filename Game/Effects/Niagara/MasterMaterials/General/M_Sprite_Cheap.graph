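# Cheap additive, unlit master material for Niagara sprite, ribbon and mesh particle effects.
# A Sprite texture, with UVs optionally blended toward another texture, supplies color or a selected channel, which multiplies particle color.
# An optional time-of-day tint from a parameter collection feeds in before the shared emissive adjustments function.
# Particle alpha multiplies either texture alpha or, via static switch, the red channel with optional inversion and subtraction.
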Material /Game/Effects/Niagara/MasterMaterials/General/M_Sprite_Cheap
graph {
 properties {
  bAutomaticallySetUsageInEditor=false
  bUsedWithNiagaraMeshParticles=true
  bUsedWithNiagaraRibbons=true
  bUsedWithNiagaraSprites=true
  blendMode=BLEND_Additive
  shadingModel=MSM_Unlit
  usedShadingModels=Unlit
 }
 nodes [
  n1 = CollectionParameter() {
   collection=&FortniteMaterialParameters
   group=ToD
   parameterName=FXTimeOfDayColorTint
  }
  n2 = LinearInterpolate(
   A=Constant3Vector() {constant=<rgba 1 1 1 0>}
   B=ComponentMask(None=@n1) {b=true, g=true, r=true}
   Alpha=ScalarParameter() {defaultValue=1, group=ToD, parameterName=ToD Contribution, sortPriority=33}
  )
  n3 = ParticleColor()
  n4 = ScalarParameter() {parameterName=Sprite Emissive - ChannelSelect?}
  n5 = TextureSampleParameter2D(UVs=LinearInterpolate(
    A=TextureCoordinate()
    B=TextureSampleParameter2D() {parameterName=TextureForUVs, samplerType=SAMPLERTYPE_LinearColor, texture=/Game/Effects/Fort_Effects/Textures/HeightForWater}.A
    Alpha=ScalarParameter() {parameterName=UseTextureAsUVs?}
   )) {group=Texture, parameterName=Sprite, texture=/Game/Effects/Fort_Effects/Textures/Sparks/T_Sparkles_4Star_01_D}
  n6 = ChannelMaskParameter(None=@n5.RGB) {parameterName=Emissive - ChannelSelect}
  n7 = Multiply(
   A=@n3.RGB
   B=LinearInterpolate(
    A=Constant() {r=1}
    B=If(A=@n4, "A > B"=@n6.R, "A == B"=@n5.RGB, "A < B"=@n5.RGB)
    Alpha=ScalarParameter() {defaultValue=1, parameterName=UseTextureColor?}
   )
  )
  MaterialFunctionCall(Emissive=StaticSwitchParameter(True=Multiply(A=@n2, B=@n7), False=@n7) {group=ToD, parameterName=Use FXTimeOfDayColorTint MPC, sortPriority=31}) {materialFunction=&MF_FX_EmissiveAdjustments}
  n8 = Saturate(None=Subtract(A=OneMinus(None=@n5.R), B=ScalarParameter() {parameterName=SubtractFromTexture}))
  n9 = ScalarParameter() {parameterName=Use OneMinus Texture?}
  StaticSwitchParameter(True=Multiply(A=@n3.A, B=LinearInterpolate(A=@n5.R, B=@n8, Alpha=@n9)), False=Multiply(A=@n3.A, B=@n5.A)) {group=Texture, parameterName=Use Texture R As Alpha, sortPriority=33}
 ]
}
