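# Material function that generates UVs for fake decals on translucent static meshes, stored under the player's Headband emote FX.
# It transforms the world position behind the translucency into the mesh's local space, then divides XY by an input bounds size, defaulting to local bounds, and offsets by 0.5.
# A second Mask output zeroes areas outside the 0-1 UV range using saturate, frac and ceil, multiplying both channels.
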MaterialFunction /Game/Animation/Game/MainPlayer/Emotes/Headband/FX/MF_FakeDecalUVs_StaticMesh
graph {
 properties {description="Generates UVs for fake decals, for translucent static meshes.", libraryCategoriesText [- Tuatara VFX]}
 nodes [
  n1 = TransformPosition(None=MaterialFunctionCall() {materialFunction=&WorldPositionBehindTranslucency}.Result) {transformSourceType=TRANSFORMPOSSOURCE_World}
  n2 = FunctionInput(Preview=ComponentMask(None=MaterialFunctionCall() {materialFunction=&ObjectLocalBounds}.Local Bounds Size) {g=true, r=true}) {
   bUsePreviewValueAsDefault=true
   description=Defaults to mesh's local bounds XY
   inputName=Static Mesh XY Bounds
   inputType=FunctionInput_Vector2
  }
  n3 = Add(A=Divide(
    A=ComponentMask(None=@n1) {desc="This uses the mesh's XY for the projection.\r\nFor different axes, change this mask", g=true, r=true}
    B=@n2
   )) {constB=0.5}
  FunctionOutput(None=@n3) {outputName=UVs}
  n4 = Ceil(None=Frac(None=Saturate(None=@n3)))
  FunctionOutput(None=Multiply(A=ComponentMask(None=@n4) {r=true}, B=ComponentMask(None=@n4) {g=true})) {description=Masks the UVs outside the 0-1 range, outputName=Mask, sortPriority=1}
 ]
}
