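# M_Sprite_Master_v2 is an unlit, translucent, two-sided master material for Niagara sprite and mesh particles.
# Two sprite textures, with optional noise UV distortion and radial mapping, multiply with edge gradients, depth and camera fades, and noise erosion to form opacity.
# Particle color, an optional two-color gradient and an emissive adjustment function build the color.
# Four dynamic parameters drive tiling, offsets and erosion per particle, and a fake decal mode swaps in a projected texture.
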Material /Game/Effects/Niagara/MasterMaterials/General/M_Sprite_Master_v2
graph {
 properties {
  bUsedWithNiagaraMeshParticles=true
  bUsedWithNiagaraSprites=true
  blendMode=BLEND_Translucent
  pixelDepthOffsetMode=PDOM_Legacy
  shadingModel=MSM_Unlit
  twoSided=true
  usedShadingModels=Unlit
 }
 nodes [
  n1 = ParticleColor()
  n2 = VectorParameter() {parameterName=Gradient Color 01, defaultValue=<rgba 0.2831 1 0.0595 1>}
  n3 = TextureCoordinate()
  n4 = Multiply(A=@n2.RGB, B=Power(Base=ComponentMask(None=@n3) {r=true}, Exp=ScalarParameter() {defaultValue=2, parameterName=Right Color Gradient Power}))
  n5 = VectorParameter() {parameterName=Gradient Color 02, defaultValue=<rgba 0.8078 1 0.1412 1>}
  n6 = OneMinus(None=@n3)
  n7 = Multiply(A=@n5.RGB, B=Power(Base=ComponentMask(None=@n6) {r=true}, Exp=ScalarParameter() {defaultValue=2, parameterName=Left Color Gradient Power}))
  n8 = Constant() {r=0.5}
  n9 = Multiply(A=@n5.RGB, B=Power(Base=ComponentMask(None=@n3) {g=true}, Exp=ScalarParameter() {defaultValue=2, parameterName=Bottom Color Gradient Power}))
  n10 = OneMinus(None=@n3)
  n11 = Multiply(A=@n2.RGB, B=Power(Base=ComponentMask(None=@n10) {g=true}, Exp=ScalarParameter() {defaultValue=2, parameterName=Top Color Gradient Power}))
  n12 = Multiply(
   A=LinearInterpolate(
    A=LinearInterpolate(A=@n4, B=@n7, Alpha=@n8)
    B=LinearInterpolate(A=@n9, B=@n11, Alpha=@n8)
    Alpha=ScalarParameter() {parameterName=Rotate Gradient 90 Degrees?}
   )
   B=@n1.RGB
  )
  n13 = ScalarParameter() {parameterName=UseGradient}
  n14 = ParticleColor()
  n15 = ScalarParameter() {parameterName=UseFakeDecal}
  n16 = MaterialFunctionCall(Emissive=LinearInterpolate(A=LinearInterpolate(A=@n1.RGB, B=@n12, Alpha=@n13), B=@n14.RGB, Alpha=@n15)) {materialFunction=&MF_FX_EmissiveAdjustments}
  If(A=ScalarParameter() {parameterName=PaintedVariant}, "A > B"=ParticleColor().RGB, "A == B"=@n16.Result, "A < B"=@n16.Result)
  n17 = MaterialFunctionCall(Float2=LinearInterpolate(
    A=TextureCoordinate()
    B=TextureCoordinate() {coordinateIndex=1}
    Alpha=ScalarParameter() {group="5) Macro UV Distortion", parameterName=Macro UV Channel, sortPriority=1}
   )) {materialFunction=&BreakOutFloat2Components}
  n18 = ScalarParameter() {defaultValue=1, parameterName=Distortion Tile U}
  n19 = ScalarParameter() {defaultValue=1, parameterName=Distortion Tile V}
  n20 = TextureCoordinate()
  n21 = ScalarParameter() {parameterName="Distortion Rotation Angle (0-1)"}
  n22 = TextureCoordinate() {coordinateIndex=1}
  n23 = Subtract(A=LinearInterpolate(
    A=MaterialFunctionCall(UVs=@n20, "Rotation Angle (0-1)"=@n21) {materialFunction=&CustomRotator}.Rotated Values
    B=MaterialFunctionCall(UVs=@n22, "Rotation Angle (0-1)"=@n21) {materialFunction=&CustomRotator}.Rotated Values
    Alpha=ScalarParameter() {group=BaseTexture, parameterName=Use_UV_1, sortPriority=-1}
   )) {constB=0.5}
  n24 = MaterialFunctionCall(Float2=@n23) {materialFunction=&BreakOutFloat2Components}
  n25 = Arctangent2Fast(Y=@n24.G, X=@n24.G)
  n26 = MaterialFunctionCall() {materialFunction=&Pi}
  n27 = Constant() {r=2}
  n28 = AppendVector(
   A=SquareRoot(None=MaterialFunctionCall(Vector 2=@n23) {materialFunction=/Engine/Functions/Engine_MaterialFunctions02/Utility/VectorLength}.V2 Length)
   B=Divide(
    A=Add(A=@n25, B=@n26.Result)
    B=MaterialFunctionCall(Multiplier=@n27) {materialFunction=&Pi}.Result
   )
  )
  n29 = LinearInterpolate(
   A=AppendVector(A=Multiply(A=@n17.R, B=@n18), B=Multiply(A=@n17.G, B=@n19))
   B=@n28
   Alpha=ScalarParameter() {parameterName=Use Radial Texture for Distortion}
  )
  n30 = DynamicParameter() {parameterIndex=3, defaultValue=<rgba 1 0 1 1>, paramNames [Erosion Power, Erosion Hardness, Distortion Offset U, Distortion Offset V]}
  n31 = AppendVector(A=@n30.Distortion Offset U, B=@n30.Distortion Offset U)
  n32 = VectorParameter() {group="5) Macro UV Distortion", parameterName=Distortion Speed, sortPriority=1}
  n33 = TextureSampleParameter2D(UVs=Panner(Coordinate=Add(A=@n29, B=@n31), Speed=AppendVector(A=@n32.R, B=@n32.R))) {
   desc=Texture for large uv distortions.
   group="5) Macro UV Distortion"
   parameterName=Distortion Texture
   samplerSource=SSM_Wrap_WorldGroupSettings
   sortPriority=0
   texture=&AtmosphericCloudNoise01
  }
  n34 = Multiply(A=AppendVector(A=@n33.R, B=Constant())) {constB=2}
  n35 = Multiply(A=AppendVector(A=Constant(), B=@n33.R)) {constB=2}
  n36 = ScalarParameter() {
   defaultValue=1
   desc=Lerp between the different directions when is Grey Scale is 1
   group="5) Macro UV Distortion"
   parameterName="Direction (U or V)"
   sortPriority=8
  }
  n37 = ScalarParameter() {defaultValue=1, desc=Set to 1 if your texture is mono color., group="5) Macro UV Distortion", parameterName=is GreyScale, sortPriority=2}
  n38 = LinearInterpolate(A=AppendVector(A=@n33.R, B=@n33.R), B=LinearInterpolate(A=Add(A=@n34) {constB=-1}, B=Add(A=@n35) {constB=-1}, Alpha=@n36), Alpha=@n37)
  n39 = ScalarParameter() {defaultValue=1, parameterName=Distortion Amount}
  n40 = StaticSwitchParameter(True=ShadingPathSwitch(Default=Multiply(A=@n38, B=@n39), Mobile=Constant2Vector()), False=Constant()) {parameterName=UseDistortion}
  n41 = TextureCoordinate()
  n42 = ScalarParameter() {parameterName="Sprite 02 Rotation Angle (0-1)"}
  n43 = TextureCoordinate() {coordinateIndex=1}
  n44 = MaterialFunctionCall(Float2=LinearInterpolate(
    A=MaterialFunctionCall(UVs=@n41, "Rotation Angle (0-1)"=@n42) {materialFunction=&CustomRotator}.Rotated Values
    B=MaterialFunctionCall(UVs=@n43, "Rotation Angle (0-1)"=@n42) {materialFunction=&CustomRotator}.Rotated Values
    Alpha=ScalarParameter() {group=BaseTexture, parameterName=Use_UV_1, sortPriority=-1}
   )) {materialFunction=&BreakOutFloat2Components}
  n45 = DynamicParameter() {parameterIndex=1, defaultValue=<rgba 1 1 0 0>, paramNames [Sprite 02 Tile U, Sprite 02 Tile V, Sprite 02 Offset U, Sprite 02 Offset V]}
  n46 = AppendVector(A=Multiply(A=@n44.R, B=@n45.Sprite 02 Tile U), B=Multiply(A=@n44.G, B=@n45.Sprite 02 Tile V))
  n47 = ScalarParameter() {parameterName=Use Radial Texture for Sprite 02}
  n48 = Add(A=@n40, B=Add(A=LinearInterpolate(A=@n46, B=@n28, Alpha=@n47), B=AppendVector(A=@n45.Sprite 02 Offset U, B=@n45.Sprite 02 Offset U)))
  n49 = TextureCoordinate()
  n50 = ScalarParameter() {parameterName="Sprite 01 Rotation Angle (0-1)"}
  n51 = TextureCoordinate() {coordinateIndex=1}
  n52 = MaterialFunctionCall(Float2=LinearInterpolate(
    A=MaterialFunctionCall(UVs=@n49, "Rotation Angle (0-1)"=@n50) {materialFunction=&CustomRotator}.Rotated Values
    B=MaterialFunctionCall(UVs=@n51, "Rotation Angle (0-1)"=@n50) {materialFunction=&CustomRotator}.Rotated Values
    Alpha=ScalarParameter() {group=BaseTexture, parameterName=Use_UV_1, sortPriority=-1}
   )) {materialFunction=&BreakOutFloat2Components}
  n53 = DynamicParameter() {defaultValue=<rgba 1 1 0 0>, paramNames [Sprite 01 Tile U, Sprite 01 Tile V, Sprite 01 Offset U, Sprite 01 Offset V]}
  n54 = Multiply(A=AppendVector(A=@n53.Sprite 01 Tile V, B=@n53.Sprite 01 Tile V), B=@n28)
  n55 = LinearInterpolate(
   A=AppendVector(A=Multiply(A=@n52.R, B=@n53.Sprite 01 Tile U), B=Multiply(A=@n52.G, B=@n53.Sprite 01 Tile V))
   B=LinearInterpolate(
    A=@n54
    B=MaterialFunctionCall(XY=@n54) {materialFunction=&Swizzle}.YX
    Alpha=ScalarParameter() {parameterName=UseSwizzle}
   )
   Alpha=ScalarParameter() {parameterName=Use Radial Texture for Sprite 01}
  )
  n56 = AppendVector(A=@n53.Sprite 01 Offset U, B=@n53.Sprite 01 Offset U)
  n57 = TextureSampleParameter2D(UVs=Add(A=@n40, B=Add(A=@n55, B=@n56))) {parameterName=Base Sprite 01, texture=/Game/Effects/Fort_Effects/Textures/Tile/T_Tile_Texture_Cartoon_Flames}
  n58 = Multiply(
   A=@n57.RGB
   B=LinearInterpolate(
    A=Constant() {r=1}
    B=TextureSampleParameter2D(UVs=@n48) {parameterName=Base Sprite 02, texture=/Game/Effects/Fort_Effects/Textures/T_Hightower_Ribbon_Dots}.RGB
    Alpha=ScalarParameter() {parameterName=Use Sprite 02}
   )
  )
  n59 = TextureCoordinate()
  n60 = ComponentMask(None=@n59) {g=true}
  n61 = Power(Base=@n60, Exp=ScalarParameter() {defaultValue=1, parameterName=Up-Facing Gradient Threshhold})
  n62 = LinearInterpolate(
   A=Constant() {r=1}
   B=Power(Base=OneMinus(None=@n60), Exp=ScalarParameter() {defaultValue=1, parameterName=Down-Facing Gradient Threshhold})
   Alpha=ScalarParameter() {defaultValue=1, parameterName=Add Down-Facing Gradient}
  )
  n63 = Constant() {r=1}
  n64 = ComponentMask(None=@n59) {r=true}
  n65 = Power(Base=@n64, Exp=ScalarParameter() {defaultValue=1, parameterName=Left-Facing Gradient Threshhold})
  n66 = ScalarParameter() {defaultValue=1, parameterName=Add Left-Facing Gradient}
  n67 = OneMinus(None=@n64)
  n68 = ScalarParameter() {defaultValue=1, parameterName=Right-Facing Gradient Threshhold}
  n69 = Multiply(
   A=Multiply(A=Multiply(A=@n61, B=@n62), B=LinearInterpolate(A=@n63, B=@n65, Alpha=@n66))
   B=LinearInterpolate(A=Constant() {r=1}, B=Power(Base=@n67, Exp=@n68), Alpha=ScalarParameter() {defaultValue=1, parameterName=Add Right-Facing Gradient})
  )
  n70 = ScalarParameter() {defaultValue=10, parameterName=Overall Gradient Power}
  n71 = Multiply(A=@n58, B=StaticSwitchParameter(True=Multiply(A=@n69, B=@n70), False=Constant() {r=1}) {parameterName=Use Gradients})
  n72 = DepthFade(Opacity=@n71, FadeDistance=ScalarParameter() {defaultValue=1, parameterName=Depth Fade Distance})
  n73 = ScalarParameter() {group="4) Camera Fade", parameterName=Camera Fade Distance, sortPriority=4}
  n74 = ScalarParameter() {group="4) Camera Fade", parameterName=Camera Fade Offset, sortPriority=5}
  n75 = Multiply(
   A=ShadingPathSwitch(Default=@n72, Deferred=@n72, Forward=@n72, Mobile=@n71)
   B=Power(
    Base=MaterialFunctionCall(Fade Length=@n73, Fade Offset=@n74) {materialFunction=&CameraDepthFade}.Result
    Exp=ScalarParameter() {defaultValue=1, group="4) Camera Fade", parameterName=Camera Fade Exponent, sortPriority=6}
   )
  )
  n76 = MaterialFunctionCall(Float2=LinearInterpolate(A=TextureCoordinate(), B=TextureCoordinate() {coordinateIndex=1}, Alpha=ScalarParameter() {parameterName=Use UV1 for Erosion Noise})) {materialFunction=&BreakOutFloat2Components}
  n77 = DynamicParameter() {
   parameterIndex=2
   defaultValue=<rgba 1 1 0 0>
   paramNames [Erosion Noise Tile U, Erosion Noise Tile V, Erosion Noise Offset U, Erosion Noise Offset V]
  }
  n78 = AppendVector(A=Multiply(A=@n76.R, B=@n77.Erosion Noise Tile U), B=Multiply(A=@n76.G, B=@n77.Erosion Noise Tile V))
  n79 = ScalarParameter() {parameterName=Use Radial Texture for Erosion Noise}
  n80 = TextureSampleParameter2D(UVs=Add(A=LinearInterpolate(A=@n78, B=@n28, Alpha=@n79), B=AppendVector(A=@n77.Erosion Noise Offset U, B=@n77.Erosion Noise Offset U))) {
   group=Erosion
   parameterName=ErosionNoise
   samplerType=SAMPLERTYPE_Masks
   sortPriority=1
   texture=/Game/Global/Textures/Default/Noise/T_CellularNoise
  }
  n81 = DynamicParameter() {parameterIndex=3, defaultValue=<rgba 1 0 1 1>, paramNames [Erosion Power, Erosion Hardness, Distortion Offset U, Distortion Offset V]}
  n82 = Power(Base=@n80.RGB, Exp=@n81.Erosion Power)
  n83 = Clamp(None=Subtract(A=Add(A=@n82), B=@n81.Erosion Hardness))
  n84 = MaterialFunctionCall!(Static Mesh XY Bounds) {materialFunction=/Game/Animation/Game/MainPlayer/Emotes/Headband/FX/MF_FakeDecalUVs_StaticMesh}
  n85 = VectorParameter() {parameterName=Decal_UVs, defaultValue=<rgba 1 1 0 1>}
  n86 = ComponentMask(None=@n85.RGB) {g=true, r=true}
  n87 = TextureSampleParameter2D(UVs=Add(A=Multiply(A=@n84.UVs, B=@n86), B=AppendVector(A=@n85.B, B=@n85.B))) {parameterName=DecalTexture, texture=/Game/Animation/Game/MainPlayer/Emotes/Headband/FX/T_FX_Headband_Emote}
  LinearInterpolate(A=Multiply(A=@n1.A, B=Multiply(A=@n75, B=@n83)), B=Multiply(A=Multiply(A=@n87.G, B=@n84.Mask), B=@n14.A), Alpha=@n15)
  MaterialFunctionCall(Alpha Threshold=@n58, Random=ScalarParameter() {parameterName=MaskedDither}) {materialFunction=&DitherTemporalAA}
  NamedRerouteDeclaration(None=@n57.RGB) {name=MainTex, nodeColor=<rgba 0.9453 0 1 1>}
 ]
}
